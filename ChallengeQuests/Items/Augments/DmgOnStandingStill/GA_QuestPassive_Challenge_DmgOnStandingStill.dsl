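; Challenge quest passive ability that applies a standing-still damage gameplay effect to the owning player while they stay in place.
; On activation it reads a movement tolerance from a curve table and starts a looping quarter-second timer that compares capsule locations between ticks.
; Movement tags or displacement beyond tolerance remove the effect by tag; otherwise it is applied unless a check tag is present.
; Ending the ability clears the timer and effect.
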(event Ability|EventFailedToActivatePassiveAbility
  (params (ActorInfo :type Gameplay%20Ability%20Actor%20Info%20Structure))
  (|CallParentFunction ActorInfo))

(event Ability|EventActivateAbility
  (|CallParentFunction)
  (bind _outxy (Fortnite|EvaluateCurveTableRow (Variables|Default|GetCurve%20Table%20Row%20Handle)))
  (Variables|Default|SetMovementTolerance _outxy)
  (Variables|Default|SetCurrent%20location (Transformation|GetWorldLocation (Variables|/Script/Engine.Character|GetCapsuleComponent (Variables|GAVars|Data|GetOwningPlayerPawn))))
  (bind _returnvalue (Utilities|Time|SetTimerbyEvent (delegate Custom|MovingCheck) 0.25 true true))
  (Variables|Default|SetTimer%20Handle _returnvalue))

(event Custom|MovingCheck
  (Variables|Default|SetNewLocation (Transformation|GetWorldLocation (Variables|/Script/Engine.Character|GetCapsuleComponent (Variables|GAVars|Data|GetOwningPlayerPawn))))
  ;[0.2083 0.9706 1] Apply or remove the GE
  (CallFunction|UpdateEffectState)
  (Variables|Default|SetCurrent%20location (Variables|Default|GetNewLocation)))

(event Ability|EventOnEndAbility
  (params (bWasCancelled :type Boolean))
  (|CallParentFunction bWasCancelled)
  (Utilities|Time|ClearandInvalidateTimerbyHandle (Variables|Default|GetTimer%20Handle))
  (GameplayEffects|RemoveActiveEffectswithTags (Ability|GetAbilitySystemComponent (Variables|GAVars|Data|GetOwningPlayerPawn)) (Variables|Default|GetT_BeingDMGd)))

(fn UpdateEffectState
  ;[0.2083 0.9706 1] Caching so we don't have to find it twice
  (LocalVariables|SetAbilityComponentSystem (Ability|GetAbilitySystemComponent (Variables|GAVars|Data|GetOwningPlayerPawn)))
  (if (not (GameplayTags|HasAnyMatchingGameplayTags (LocalVariables|GetAbilityComponentSystem) (Variables|Default|GetT_MovementTags)))
    (if (Math|Vector|VectorIsNearlyZero (- (Variables|Default|GetNewLocation) (Variables|Default|GetCurrent%20location)) (Variables|Default|GetMovementTolerance))
      (if (not (GameplayTags|HasMatchingGameplayTag (LocalVariables|GetAbilityComponentSystem) (Variables|Default|GetCheckDmgTag)))
        (GameplayEffects|ApplyGameplayEffectToSelf :self (LocalVariables|GetAbilityComponentSystem) :GameplayEffectClass (Variables|Default|GetGE_StandingStillDMG) :Level 0.0))
      (else
        _))
    (else
      _))
  (GameplayEffects|RemoveActiveEffectswithTags (LocalVariables|GetAbilityComponentSystem) (Variables|Default|GetT_BeingDMGd)))
